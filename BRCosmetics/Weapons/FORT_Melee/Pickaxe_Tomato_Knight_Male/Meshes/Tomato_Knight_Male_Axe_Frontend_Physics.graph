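# Physics asset for the frontend mesh of the male Tomato Knight pickaxe, defining bodies for its dynamic blade and crossguard bones.
# A box covers the blade, which hinges on a small master capsule, free on one swing axis with the other swing and twist locked.
# Four crossguard capsules each attach to their own master capsule with moderate swing limits and a narrow twist limit.
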PhysicsAsset /BRCosmetics/Weapons/FORT_Melee/Pickaxe_Tomato_Knight_Male/Meshes/Tomato_Knight_Male_Axe_Frontend_Physics
bodies {
 dyn_blade_01 [{shapeType=Box, extentX=40.8535, extentY=2.6259, extentZ=40.8685, center=<-3.5 0 -5>}]
 dyn_blade_MSTR [{shapeType=Capsule, length=0.505, radius=0.505}]
 dyn_crossguard_01 [{shapeType=Capsule, length=2.755, radius=1.255, center=<2.5 0 0>, rotation=<rot 90 0 0>}]
 dyn_crossguard_02 [{shapeType=Capsule, length=2.755, radius=1.255, center=<2.5 0 0>, rotation=<rot 90 0 0>}]
 dyn_crossguard_03 [{shapeType=Capsule, length=2.755, radius=1.255, center=<2.5 0 0>, rotation=<rot 90 0 0>}]
 dyn_crossguard_04 [{shapeType=Capsule, length=2.755, radius=1.255, center=<2.5 0 0>, rotation=<rot 90 0 0>}]
 dyn_crossguard_MSTR [{shapeType=Capsule, length=0.505, radius=0.505}]
}
constraints [
 {
  bone1Name=dyn_blade_01
  bone2Name=dyn_blade_MSTR
  swing1LimitDegrees=1
  swing1Motion=Locked
  swing2LimitDegrees=180
  swing2Motion=Free
  twistLimitDegrees=1
  twistMotion=Locked
 }
 {bone1Name=dyn_crossguard_02, bone2Name=dyn_crossguard_MSTR, swing1LimitDegrees=45, swing2LimitDegrees=45, twistLimitDegrees=3}
 {bone1Name=dyn_crossguard_03, bone2Name=dyn_crossguard_MSTR, swing1LimitDegrees=45, swing2LimitDegrees=45, twistLimitDegrees=3}
 {bone1Name=dyn_crossguard_04, bone2Name=dyn_crossguard_MSTR, swing1LimitDegrees=45, swing2LimitDegrees=45, twistLimitDegrees=3}
 {bone1Name=dyn_crossguard_01, bone2Name=dyn_crossguard_MSTR, swing1LimitDegrees=45, swing2LimitDegrees=45, twistLimitDegrees=3}
]
