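# Niagara ribbon trail for the Tomato Knight (male) pickaxe, with a short particle lifetime and a pale blue emitter material color.
# It spawns particles per unit of distance moved and samples two bone or socket points on the user-supplied skeletal mesh.
# A local AnimTrail_Spawn_Location module takes both points, and their difference sets the ribbon facing, orienting the trail along the line between them.
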NiagaraSystem /BRCosmetics/Weapons/FORT_Melee/Pickaxe_Tomato_Knight_Male/VFX/NS_Pickaxe_TomatoKnight_AnimTrail
dependencies {usedDataInterfaces [], usedDynamicInputs [], usedModules [&SystemState]}
emitters {AnimTrail {
  emitterSpawnScript [/SetVariables(Emitter.MaterialColor:LinearColor=<rgba 0.3646 0.6369 1 1>) {isSetParameters=true}]
  emitterUpdateScript [
   &EmitterState()
   &SpawnPerUnit(Module.Use Max Movement Threshold=true, Module.Use Movement Tolerance=false, Spawn Per Unit=35, Use Spawn Probability=false)
  ]
  particleSpawnScript [
   &InitializeRibbon(
    Color=<rgba 1 1 1 0.35>
    Lifetime=0.165
    Module.Write Color=true
    Module.Write Lifetime=true
    Module.Write Mass=true
    Module.Write Material Random=true
    Module.Write Position=false
    Module.Write RibbonTwist=false
    Module.Write RibbonWidth=true
    Ribbon Width=100
   )
   &SkeletalMeshLocation(
    Apply To Particle Position=false
    Inherit Velocity=false
    MaskSpawnGroup=false
    Mesh Sampling Type=::NewEnumerator5
    Skeletal Mesh=$User.SkeletalMesh
    Socket Sampling Mode=::NewEnumerator5
    Use Sampled Mesh Attribute as Ribbon ID=false
   )
   &SkeletalMeshLocation(
    Apply To Particle Position=false
    Bone / Socket Index=1
    Inherit Velocity=false
    MaskSpawnGroup=false
    Mesh Sampling Type=::NewEnumerator5
    Skeletal Mesh=$User.SkeletalMesh
    Socket Sampling Mode=::NewEnumerator5
    Use Sampled Mesh Attribute as Ribbon ID=false
   ) {module=SkeletalMeshLocation001}
   &self:AnimTrail_Spawn_Location(New Vector:Vector3f=$Output.SkeletalMeshLocation.SampledPosition, New Vector001:Vector3f=$Output.SkeletalMeshLocation001.SampledPosition) {module=AnimTrail_Spawn_Location}
   /SetVariables(Particles.RibbonFacing=&Subtract_Vector(A=$Output.SkeletalMeshLocation.SampledPosition, B=$Output.SkeletalMeshLocation001.SampledPosition)) {isSetParameters=true}
  ]
  particleUpdateScript [
   &ParticleState()
   &ScaleRibbonWidth()
  ]
  rendererClasses [&NiagaraRibbonRendererProperties]
  renderers [{rendererClass=&NiagaraRibbonRendererProperties}]
 }}
summary {
 systemName=NS_Pickaxe_TomatoKnight_AnimTrail
 emitters [{emitterName=AnimTrail, rendererClasses [&NiagaraRibbonRendererProperties]}]
 userVariables [
  User.RibbonIDOffset {type=NiagaraInt32}
  User.SkeletalMesh {
   type=NiagaraDataInterfaceSkeletalMesh
   defaultValue {
    struct=NiagaraExt_VariableValue_DataInterface
    value {dataInterface=&self:NiagaraDataInterfaceSkeletalMesh_0}
   }
  }
 ]
}
